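# Physics asset for the CarbideBlue two-handed pickaxe skeletal mesh, giving its handle and two ring bones simulated collision.
# Three capsule bodies cover the handle, blade ring and handle ring.
# Each ring is constrained directly to the handle with twist and swing1 locked, leaving swing2 as the only unlocked angular axis so it pivots like a hinge.
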PhysicsAsset /BRCosmetics/Weapons/FORT_Melee/Pickaxe_CarbideBlue/Models/SK_Pickaxe_CarbideBlue_Physics
bodies {
 Melee_TwoHanded_Handle [{shapeType=Capsule, length=7.8863, radius=6.4285, center=<0.7562 8.5377 23.6757>, rotation=<rot -4.6294 -90 0.0047>}]
 blade_ring [{shapeType=Capsule, length=3.6637, radius=6, center=<-0.0228 3.8113 -2.4319>, rotation=<rot 0 0.3424 -57.4603>}]
 handle_ring [{shapeType=Capsule, length=5.451, radius=3.2127, center=<-0.0228 3.8114 -2.4319>, rotation=<rot 0 0.3424 -57.4603>}]
}
constraints [
 {
  bone1Name=handle_ring
  bone2Name=Melee_TwoHanded_Handle
  swing1LimitDegrees=45
  swing1Motion=Locked
  swing2LimitDegrees=90
  twistLimitDegrees=45
  twistMotion=Locked
 }
 {
  bone1Name=blade_ring
  bone2Name=Melee_TwoHanded_Handle
  swing1LimitDegrees=45
  swing1Motion=Locked
  swing2LimitDegrees=90
  twistLimitDegrees=45
  twistMotion=Locked
 }
]
