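# Animation Blueprint for the CarbideBlue pickaxe cosmetic, adding simulated secondary motion to the weapon's skeleton.
# The live path takes the input pose through a RigidBody node using the pickaxe's override physics asset, with amplified component linear velocity, to the output.
# Two chained world-space Anim Dynamics nodes for handle_ring and blade_ring exist but do not connect to the output.
# The event graph contains no connected logic.
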AnimBlueprint /BRCosmetics/Weapons/FORT_Melee/Pickaxe_CarbideBlue/Models/SK_Pickaxe_CarbideBlue_AnimBP
target_skeleton=&/SK_Pickaxe_CarbideBlue_Skeleton
graphs [
 AnimGraph {nodes [
   n1 = Anim Dynamics - Bone: handle_ring(Alpha=1, ExternalForce="0,0,0") {
    bGravityOverrideInSimSpace=true
    lODThreshold=1
    numSolverIterationsPostUpdate=4
    numSolverIterationsPreUpdate=16
    simulationSpace=World
    alphaBoolBlend {blendOption=Linear}
    boundBone {boneName=handle_ring}
    physicsBodyDefinitions [{
      sphereCollisionRadius=0
      boundBone {boneName=handle_ring}
      boxExtents=<2.5 2.5 2.5>
      constraintSetup {angularTargetAxis=AxisZ, twistAxis=AxisZ, angularLimitsMax=<89 0 0>, angularLimitsMin=<-89 0 0>, angularTarget=<0 0 0>}
      localJointOffset=<0 0 0>
     }]
   }
   Anim Dynamics - Bone: blade_ring(ComponentPose=@n1, Alpha=1, ExternalForce="0,0,0") {
    angularDampingOverride=0.98
    bGravityOverrideInSimSpace=true
    bOverrideAngularDamping=true
    simulationSpace=World
    alphaBoolBlend {blendOption=Linear}
    boundBone {boneName=blade_ring}
    physicsBodyDefinitions [{
      sphereCollisionRadius=0
      boundBone {boneName=blade_ring}
      boxExtents=<10 10 30>
      constraintSetup {twistAxis=AxisZ, angularLimitsMax=<89 0 0>, angularLimitsMin=<-89 0 0>, angularTarget=<0 0 0>}
      localJointOffset=<0 0 -5>
     }]
   }
   AnimGraphNode_SubInput_0 = Input Pose()
   n2 = RigidBody(ComponentPose=Local To Component(LocalPose=@AnimGraphNode_SubInput_0)) {
    overridePhysicsAsset=&/SK_Pickaxe_CarbideBlue_Physics
    alphaBoolBlend {blendOption=Linear}
    componentLinearAccScale=<1 1 1>
    componentLinearVelScale=<10 10 10>
   }
   Output Pose(Result=Component To Local(ComponentPose=@n2)) {layerGroup=None}
  ]}
 EventGraph {nodes [
   TryGetPawnOwner()
   Event BlueprintUpdateAnimation()
  ]}
]
variables []
